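# Parent material for the BinGrass gadget prop, built on the 2023 character PreFX and PostFX functions, that adds tank-tread UV panning.
# A step mask on U, sized by TreadsMaskSize, limits vertical panning to the high-U strip of the UVs.
# A wrapped time-times-TreadsSpeed value drives the offset, while floored TreadsStaticLerping switches it to a constant, static by default.
# These UVs override PreFX's texture-set and tech-art-mask UVs before PostFX.
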Material /BRCosmetics/Animation/Props/Gadgets_BinGrass/Materials/M_Gadget_BinGrass_PARENT
graph {
 properties {
  bUseMaterialAttributes=true
  bUsedWithMeshDeformer=true
  bUsedWithMorphTargets=true
  bUsedWithSkeletalMesh=true
  pixelDepthOffsetMode=PDOM_Legacy
  shadingModel=MSM_FromMaterialExpression
  usedShadingModels=Default Lit | Subsurface | Clear Coat | Subsurface Profile | Cloth
 }
 nodes [
  n1 = Constant()
  n2 = Time()
  n3 = ScalarParameter() {defaultValue=0.17, group=TankTreadPanning, parameterName=TreadsSpeed, sortPriority=0}
  n4 = LinearInterpolate(
   A=Frac(None=Multiply(A=@n2, B=@n3) {constB=0.1})
   Alpha=Floor(None=ScalarParameter() {defaultValue=1, group=TankTreadPanning, parameterName=TreadsStaticLerping, sortPriority=0})
  )
  n5 = ComponentMask(None=TextureCoordinate()) {r=true}
  n6 = OneMinus(None=ScalarParameter() {defaultValue=0.17, group=TankTreadPanning, parameterName=TreadsMaskSize, sortPriority=1})
  n7 = LinearInterpolate(
   A=TextureCoordinate()
   B=Panner(Time=Constant() {r=1}, Speed=AppendVector(A=@n1, B=@n4)) {speedY=0.1}
   Alpha=OneMinus(None=Step(Y=@n5, X=@n6) {constX=0.862})
  )
  n8 = MaterialFunctionCall(Overide Texture Set UVs=@n7, Override Tech Art Mask UVs=@n7) {materialFunction=&MF_CharacterParent_2023_PreFX}
  MaterialFunctionCall(Material=@n8.Result, M=@n8.Result, Pre Skinned Local Position=@n8.Result, Distance Blend=@n8.Result, Fresnel=@n8.Result) {materialFunction=&MF_CharacterParent_2023_PostFX}
 ]
}
